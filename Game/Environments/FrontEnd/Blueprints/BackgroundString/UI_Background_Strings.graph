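# UI_Background_Strings is a front-end background widget that shows a stylized, configurable title string, with "Fortnite Crew" as its designer placeholder.
# Both PreConstruct and Construct copy the CustomText variable into a CommonRichTextBlock, which supports inline size tags.
# A down-only ScaleBox inside a fixed-size SizeBox shrinks long text to fit.
# The text is centered, uppercase white Burbank with tight line spacing and a sheared render transform for a slanted look.
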WidgetBlueprint /Game/Environments/FrontEnd/Blueprints/BackgroundString/UI_Background_Strings
graphs [EventGraph {nodes [
   n1 = Event PreConstruct()
   SetText(
    execute=[@n1.then, @n2.then]
    self=Get CustomString() -> (CustomString:Common Rich Text Block Object Reference)
    InText=Get CustomText() -> (CustomText:Text)
   )
   n2 = Event Construct()
  ]}]
variables [CustomText {type=string, default=""}]
widgets {
 info {inheritedWidgetCount=0, parentClass=&UserWidget, rootWidgetClass=&SizeBox, widgetCount=3}
 widgets [
  SizeBox_11 {
   widgetClassPath=&SizeBox
   widget=SizeBox_11
   properties {bOverride_HeightOverride=true, bOverride_WidthOverride=true, heightOverride=256, widthOverride=1024}
  }
  ScaleBox_0 {widgetClassPath=&ScaleBox, parent=SizeBox_11, widget=ScaleBox_0, properties {stretchDirection=DownOnly}, slot {class=SizeBoxSlot}}
  CustomString {
   widgetClassPath=&CommonRichTextBlock
   bIsVariable=true
   parent=ScaleBox_0
   widget=CustomString
   properties {
    bOverrideDefaultStyle=true
    justification=Center
    lineHeightPercentage=0.5
    text="<text size=\"48\">Fortnite</>\r\n<text size=\"128\">Crew</>"
    textTransformPolicy=ToUpper
    defaultTextStyleOverride {
     overflowPolicy=Clip
     transformPolicy=None
     colorAndOpacity {colorUseRule=UseColor_Specified, specifiedColor=<rgba 1 1 1 1>}
     font {
      bForceMonospaced=false
      bMaterialIsStencil=false
      fontMaterial=null
      fontObject=&Burbank
      letterSpacing=0
      monospacedWidth=1
      size=96
      skewAmount=0
      typefaceFontName=BurbankRegular-Black
      outlineSettings {
       bApplyOutlineToDropShadows=false
       bMiteredCorners=false
       bSeparateFillAlpha=false
       outlineMaterial=null
       outlineSize=0
       outlineColor=<rgba 0 0 0 1>
      }
     }
     highlightColor {colorUseRule=UseColor_Specified, specifiedColor=<rgba 0 0 0 1>}
     highlightShape {
      bIsDynamicallyLoaded=false
      drawAs=Image
      imageType=NoImage
      mirroring=NoMirror
      resourceName=None
      resourceObject=null
      tiling=NoTile
      imageSize=<32 32>
      margin=<ltrb 0 0 0 0>
      outlineSettings {
       bUseBrushTransparency=false
       roundingType=HalfHeightRadius
       width=0
       color {colorUseRule=UseColor_Specified, specifiedColor=<rgba 0 0 0 0>}
       cornerRadii=<0 0 0 0>
      }
      tintColor {colorUseRule=UseColor_Specified, specifiedColor=<rgba 1 1 1 1>}
      uVRegion {bIsValid=false, max=<0 0>, min=<0 0>}
     }
     selectedBackgroundColor {colorUseRule=UseColor_Foreground, specifiedColor=<rgba 1 0 1 1>}
     shadowColorAndOpacity=<rgba 0 0 0 1>
     shadowOffset=<0 0>
     strikeBrush {
      bIsDynamicallyLoaded=false
      drawAs=Image
      imageType=NoImage
      mirroring=NoMirror
      resourceName=None
      resourceObject=null
      tiling=NoTile
      imageSize=<32 32>
      margin=<ltrb 0 0 0 0>
      outlineSettings {
       bUseBrushTransparency=false
       roundingType=HalfHeightRadius
       width=0
       color {colorUseRule=UseColor_Specified, specifiedColor=<rgba 0 0 0 0>}
       cornerRadii=<0 0 0 0>
      }
      tintColor {colorUseRule=UseColor_Specified, specifiedColor=<rgba 1 1 1 1>}
      uVRegion {bIsValid=false, max=<0 0>, min=<0 0>}
     }
     underlineBrush {
      bIsDynamicallyLoaded=false
      drawAs=Image
      imageType=NoImage
      mirroring=NoMirror
      resourceName=None
      resourceObject=null
      tiling=NoTile
      imageSize=<32 32>
      margin=<ltrb 0 0 0 0>
      outlineSettings {
       bUseBrushTransparency=false
       roundingType=HalfHeightRadius
       width=0
       color {colorUseRule=UseColor_Specified, specifiedColor=<rgba 0 0 0 0>}
       cornerRadii=<0 0 0 0>
      }
      tintColor {colorUseRule=UseColor_Specified, specifiedColor=<rgba 1 1 1 1>}
      uVRegion {bIsValid=false, max=<0 0>, min=<0 0>}
     }
    }
    margin=<ltrb 32 -48 32 0>
    renderTransform {angle=0, scale=<1 1>, shear=<-12 0>, translation=<0 -32>}
   }
   slot {class=ScaleBoxSlot}
  }
 ]
}
